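# Niagara system for the Stylizedpsplus vehicle boost cosmetic, emitting flipbook sprite trails from skeletal mesh sockets on the car.
# Only NE_VC_Boost_Sprite_Parent001 is enabled; two sibling emitters are disabled.
# It spreads particles across filtered sockets, with VC_Mask inputs varying rate, lifetime and rearward velocity for Coro, FrontEnd and FN.
# Sprites briefly inherit source movement, slow under rising drag, grow, and shift from magenta through cyan to muted grey.
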NiagaraSystem /VehicleCosmetics/Boosts/Boost_Stylizedpsplus/FX/NS_Stylizedpsplus
dependencies {
 usedDataInterfaces [
  &NiagaraDataInterfaceCurve
  &NiagaraDataInterfaceSpriteRendererInfo
  &NiagaraDataInterfaceColorCurve
  &NiagaraDataInterfaceCamera
  &NiagaraDataInterfaceSkeletalMesh
  &NiagaraDataInterfaceArrayFloat3
 ]
 usedDynamicInputs [&SelectVectorFromArray]
 usedModules [
  &self:SystemSpawnScript.SystemScriptSource.SystemScriptGraph.NiagaraNodeAssignment_0.SetVariables_#1
  &self:SystemSpawnScript.SystemScriptSource.SystemScriptGraph.NiagaraNodeAssignment_1.SetVariables_#2
  &self:SystemSpawnScript.SystemScriptSource.SystemScriptGraph.NiagaraNodeAssignment_2.SetVariables_#3
  &SystemState
 ]
}
emitters {
 NE_VC_Boost {
  bEnabled=false
  emitterUpdateScript [
   &EmitterState()
   &VC_FilteredSocketCount()
   &SpawnRate(
    SpawnRate=&Multiply_Float(
     A=200
     B=/VehicleCosmetics/FX/Niagara/Modules/VC_Mask_Floats(Coro Value:NiagaraFloat=1, FrontEnd Value:NiagaraFloat=0.6, FN Value:NiagaraFloat=0.5)
    )
    Use Spawn Probability=false
   )
  ]
  particleSpawnScript [
   &VC_PaintedVariantsOverride() {enabled=false}
   &InitializeParticle(
    Color Mode=::NewEnumerator0
    Lifetime=&VC_Mask_Floats(
     Coro Value:NiagaraFloat=&RandomRangeFloat(Minimum=0.5, Maximum=0.6)
     FrontEnd Value:NiagaraFloat=&RandomRangeFloat(Minimum=0.7, Maximum=0.9)
     FN Value:NiagaraFloat=&RandomRangeFloat(Minimum=0.6, Maximum=0.8)
    )
    Sprite Rotation Mode=::NewEnumerator2
    Sprite Size Mode=::NewEnumerator1
    Uniform Sprite Size Max=170
    Uniform Sprite Size Min=140
    UsePositionOffset=false
    Write Material Random=true
   )
   &PartitionParticles(Number Of Partitions=$Emitter.PortCount)
   &SkeletalMeshLocation(
    Bone / Socket Index=$Particles.PartitionParticles.Partition
    Bone / Socket Sampling Mode=::NewEnumerator5
    MaskSpawnGroup=false
    Mesh Sampling Type=::NewEnumerator6
    Sampled Position Offset=&AddVectorToPosition(
     Position=&VC_Mask_Vectors(Coro Value:Vector3f=<-30 0 0>, FrontEnd Value:Vector3f=<-30 0 0>, FN Value:Vector3f=<-30 0 0>)
     Vector=$User.DTCarOffset
    )
    Skeletal Mesh=$System.SkeletalMesh
   )
   &AddVelocity(Velocity=&VC_Mask_Vectors(Coro Value:Vector3f=<-400 0 0>, FrontEnd Value:Vector3f=<-1800 0 0>, FN Value:Vector3f=<-1200 0 0>))
   /VehicleCosmetics/FX/VC_FXParents/Modules/DLR_ProductPreviews(GameModeID:NiagaraInt32=$System.GameModeID, bPreviewVehicleSpeed:NiagaraBool=$System.bPreviewVehicleSpeed) {enabled=false}
  ]
  particleUpdateScript [
   &ParticleState()
   &InheritSourceMovement(
    Applied Position Scale=&VectorFromFloat(Value=&FloatFromCurve(FloatCurve=DataInterface() {propertyValues {Curve {
         defaultValue=1
         keys [
          {arriveTangent=-1.7885, interpMode=RCIM_Cubic, leaveTangent=-1.7885, tangentMode=RCTM_User, time=0.05, value=1}
          {interpMode=RCIM_Cubic, tangentMode=RCTM_User, time=1, value=0}
         ]
        }}}))
    ApplyForce=false
    ApplyPosition=true
    ApplyPositionFalloff=false
   )
   &Drag(
    Drag=&FloatFromCurve(FloatCurve=DataInterface() {propertyValues {Curve {
        defaultValue=1
        keys [
         {arriveTangent=1.3812, interpMode=RCIM_Cubic, leaveTangent=1.3812, leaveTangentWeight=0.3333, tangentMode=RCTM_User, time=0, value=0}
         {arriveTangent=0.6233, arriveTangentWeight=0.3333, interpMode=RCIM_Cubic, leaveTangent=0.6233, tangentMode=RCTM_User, time=1, value=3}
        ]
       }}})
    Use Linear Drag=true
    Use Rotational Drag=false
   )
   &SolveForcesAndVelocity(Clamp Velocity=false, Limit Acceleration=false)
   &SubUVAnimation(
    "Sprite Renderer "=DataInterface() {propertyValues {SpriteRenderer=&self:NE_VC_Boost_0.NiagaraSpriteRendererProperties_0}}
    UseEndFrame=false
    UseStartFrame=false
   )
   &ScaleSpriteSize(Uniform Curve Sprite Scale=DataInterface() {propertyValues {Curve {keys [
        {arriveTangent=0.7207, interpMode=RCIM_Cubic, leaveTangent=0.7207, tangentMode=RCTM_User, time=0, value=0.98}
        {
         arriveTangent=2.5492
         arriveTangentWeight=0.277
         interpMode=RCIM_Cubic
         leaveTangent=2.5492
         leaveTangentWeight=0.3239
         tangentMode=RCTM_User
         time=0.45
         value=1.7
        }
        {arriveTangentWeight=0.1763, interpMode=RCIM_Cubic, tangentMode=RCTM_User, time=1, value=2.5}
       ]}}})
   &Color(Color=&MultiplyLinearColor(
     Linear Color=&ColorFromCurve(ColorCurve=DataInterface() {propertyValues {
        AlphaCurve {keys [<at 0 1 RCIM_Cubic>]}
        BlueCurve {keys [
          <at 0 0.95>
          <at 0.0435 0.95>
          {arriveTangent=0.4984, interpMode=RCIM_Cubic, leaveTangent=0.4984, time=0.1221, value=0.95}
          {arriveTangent=0.6187, interpMode=RCIM_Cubic, leaveTangent=0.6187, time=0.1438, value=1}
          <at 0.2029 1>
          <at 0.2499 1>
          <at 0.2937 1>
          <at 0.3527 0.1875>
         ]}
        GreenCurve {keys [
          <at 0 0.4651>
          <at 0.0435 0.076>
          {arriveTangent=9.21, interpMode=RCIM_Cubic, leaveTangent=9.21, time=0.1221, value=0.076}
          {arriveTangent=9.3705, interpMode=RCIM_Cubic, leaveTangent=9.3705, time=0.1438, value=1}
          <at 0.2029 0.8333>
          <at 0.2499 0.8333>
          <at 0.2937 1>
          <at 0.3527 0.1688>
         ]}
        RedCurve {keys [
          <at 0 0.8692>
          <at 0.0435 0.8043>
          {arriveTangent=-4.0302, interpMode=RCIM_Cubic, leaveTangent=-4.0302, time=0.1221, value=0.8772}
          {arriveTangent=-10.8532, interpMode=RCIM_Cubic, leaveTangent=-10.8532, time=0.1438, value=0.4}
          <at 0.2029 0>
          <at 0.2499 0>
          <at 0.2937 0.4>
          <at 0.3527 0.1688>
         ]}
       }})
     Scale Factor=&MakeVector4FromVectorAndFloat(
      "Vector (RGB)"=&VectorFromFloat(Value=0.8)
      "Float (Alpha)"=/Niagara/DynamicInputs/Camera/CameraFade(
       Camera Query:NiagaraDataInterfaceCamera=DataInterface()
       Fade Offset:NiagaraFloat=6
       Fade Distance:NiagaraFloat=19
       Use Owner Position:NiagaraBool=true
      )
     )
    ))
  ]
  rendererClasses [&NiagaraSpriteRendererProperties]
  renderers [
   {rendererClass=&NiagaraSpriteRendererProperties}
   {rendererClass=&NiagaraSpriteRendererProperties, rendererIndex=1}
  ]
 }
 NE_VC_Boost_Sprite_Parent {
  bEnabled=false
  emitterUpdateScript [
   &EmitterState()
   &VC_FilteredSocketCount()
   &SpawnRate(
    SpawnRate=&Multiply_Float(
     A=200
     B=&VC_Mask_Floats(Coro Value:NiagaraFloat=1, FrontEnd Value:NiagaraFloat=0.5, FN Value:NiagaraFloat=0.5)
    )
    Use Spawn Probability=false
   )
  ]
  particleSpawnScript [
   &VC_PaintedVariantsOverride() {enabled=false}
   &InitializeParticle(
    Color Mode=::NewEnumerator0
    Lifetime=0.5
    Sprite Rotation Mode=::NewEnumerator2
    Sprite Size Mode=::NewEnumerator3
    Uniform Sprite Size=0
    UsePositionOffset=false
    Write Material Random=true
   )
   &PartitionParticles(Number Of Partitions=$Emitter.PortCount)
   &SkeletalMeshLocation(
    Bone / Socket Index=$Particles.PartitionParticles.Partition
    Bone / Socket Sampling Mode=::NewEnumerator5
    MaskSpawnGroup=false
    Mesh Sampling Type=::NewEnumerator6
    Skeletal Mesh=$System.SkeletalMesh
   )
   &AddVelocity(Velocity=&VC_Mask_Vectors(Coro Value:Vector3f=<-400 0 0>, FrontEnd Value:Vector3f=<-1200 0 0>, FN Value:Vector3f=<-400 0 0>))
   &DLR_ProductPreviews(GameModeID:NiagaraInt32=$System.GameModeID, bPreviewVehicleSpeed:NiagaraBool=$System.bPreviewVehicleSpeed)
  ]
  particleUpdateScript [
   &ParticleState()
   &InheritSourceMovement(
    Applied Position Scale=&VectorFromFloat(Value=&FloatFromCurve(FloatCurve=DataInterface() {propertyValues {Curve {
         defaultValue=1
         keys [
          {arriveTangent=-1.7885, interpMode=RCIM_Cubic, leaveTangent=-1.7885, tangentMode=RCTM_User, time=0.05, value=1}
          {interpMode=RCIM_Cubic, tangentMode=RCTM_User, time=1, value=0}
         ]
        }}}))
    ApplyForce=false
    ApplyPosition=true
    ApplyPositionFalloff=false
   )
   &SolveForcesAndVelocity(Clamp Velocity=false, Limit Acceleration=false)
  ]
  rendererClasses [&NiagaraSpriteRendererProperties]
  renderers [{rendererClass=&NiagaraSpriteRendererProperties}]
 }
 NE_VC_Boost_Sprite_Parent001 {
  emitterUpdateScript [
   &EmitterState()
   &VC_FilteredSocketCount()
   &SpawnRate(
    SpawnRate=&Multiply_Float(
     A=200
     B=&VC_Mask_Floats(Coro Value:NiagaraFloat=1, FrontEnd Value:NiagaraFloat=0.6, FN Value:NiagaraFloat=0.5)
    )
    Use Spawn Probability=false
   )
  ]
  particleSpawnScript [
   &VC_PaintedVariantsOverride() {enabled=false}
   &InitializeParticle(
    Color Mode=::NewEnumerator0
    Lifetime=&VC_Mask_Floats(
     Coro Value:NiagaraFloat=&RandomRangeFloat(Minimum=0.5, Maximum=0.6)
     FrontEnd Value:NiagaraFloat=&RandomRangeFloat(Minimum=0.7, Maximum=0.8)
     FN Value:NiagaraFloat=&RandomRangeFloat(Minimum=0.6, Maximum=0.8)
    )
    Sprite Rotation Mode=::NewEnumerator2
    Sprite Size Mode=::NewEnumerator1
    Uniform Sprite Size Max=170
    Uniform Sprite Size Min=140
    UsePositionOffset=false
    Write Material Random=true
   )
   &PartitionParticles(Number Of Partitions=$Emitter.PortCount)
   &SkeletalMeshLocation(
    Bone / Socket Index=$Particles.PartitionParticles.Partition
    Bone / Socket Sampling Mode=::NewEnumerator5
    MaskSpawnGroup=false
    Mesh Sampling Type=::NewEnumerator6
    Sampled Position Offset=&AddVectorToPosition(
     Position=&VC_Mask_Vectors(Coro Value:Vector3f=<-30 0 0>, FrontEnd Value:Vector3f=<-30 0 0>, FN Value:Vector3f=<-30 0 0>)
     Vector=$User.DTCarOffset
    )
    Skeletal Mesh=$System.SkeletalMesh
   )
   &AddVelocity(Velocity=&VC_Mask_Vectors(Coro Value:Vector3f=<-400 0 0>, FrontEnd Value:Vector3f=<-1600 0 0>, FN Value:Vector3f=<-1200 0 0>))
   &DLR_ProductPreviews(GameModeID:NiagaraInt32=$System.GameModeID, bPreviewVehicleSpeed:NiagaraBool=$System.bPreviewVehicleSpeed) {enabled=false}
  ]
  particleUpdateScript [
   &ParticleState()
   &InheritSourceMovement(
    Applied Position Scale=&VectorFromFloat(Value=&FloatFromCurve(FloatCurve=DataInterface() {propertyValues {Curve {
         defaultValue=1
         keys [
          {arriveTangent=-1.7885, interpMode=RCIM_Cubic, leaveTangent=-1.7885, tangentMode=RCTM_User, time=0.05, value=1}
          {interpMode=RCIM_Cubic, tangentMode=RCTM_User, time=1, value=0}
         ]
        }}}))
    ApplyForce=false
    ApplyPosition=true
    ApplyPositionFalloff=false
   )
   &Drag(
    Drag=&FloatFromCurve(FloatCurve=DataInterface() {propertyValues {Curve {
        defaultValue=1
        keys [
         {arriveTangent=1.3812, interpMode=RCIM_Cubic, leaveTangent=1.3812, leaveTangentWeight=0.3333, tangentMode=RCTM_User, time=0, value=0}
         {arriveTangent=0.6233, arriveTangentWeight=0.3333, interpMode=RCIM_Cubic, leaveTangent=0.6233, tangentMode=RCTM_User, time=1, value=3}
        ]
       }}})
    Use Linear Drag=true
    Use Rotational Drag=false
   )
   &SolveForcesAndVelocity(Clamp Velocity=false, Limit Acceleration=false)
   &SubUVAnimation(
    "Sprite Renderer "=DataInterface() {propertyValues {SpriteRenderer=&self:NE_VC_Boost_Sprite_Parent001.NiagaraSpriteRendererProperties_2}}
    UseEndFrame=false
    UseStartFrame=false
   )
   &ScaleSpriteSize(Uniform Curve Sprite Scale=DataInterface() {propertyValues {Curve {keys [
        {arriveTangent=0.7207, interpMode=RCIM_Cubic, leaveTangent=0.7207, tangentMode=RCTM_User, time=0, value=0.98}
        {
         arriveTangent=2.5492
         arriveTangentWeight=0.277
         interpMode=RCIM_Cubic
         leaveTangent=2.5492
         leaveTangentWeight=0.3239
         tangentMode=RCTM_User
         time=0.45
         value=1.7
        }
        {arriveTangentWeight=0.1763, interpMode=RCIM_Cubic, tangentMode=RCTM_User, time=1, value=2.5}
       ]}}})
   &Color(Color=&MultiplyLinearColor(
     Linear Color=&ColorFromCurve(ColorCurve=DataInterface() {propertyValues {
        AlphaCurve {keys [<at 0 1 RCIM_Cubic>]}
        BlueCurve {keys [
          <at 0 0.95>
          <at 0.0435 0.95>
          {arriveTangent=0.4984, interpMode=RCIM_Cubic, leaveTangent=0.4984, time=0.1221, value=0.8874}
          {arriveTangent=1.3938, interpMode=RCIM_Cubic, leaveTangent=1.3938, time=0.1438, value=1}
          <at 0.2029 1>
          <at 0.2454 1>
          <at 0.2937 1>
          <at 0.3503 0.4571>
         ]}
        GreenCurve {keys [
          <at 0 0.3661>
          <at 0.0435 0.19>
          {arriveTangent=8.0737, interpMode=RCIM_Cubic, leaveTangent=8.0737, time=0.1221, value=0.1983}
          {arriveTangent=7.8578, interpMode=RCIM_Cubic, leaveTangent=7.8578, time=0.1438, value=1}
          <at 0.2029 0.8333>
          <at 0.2454 0.8333>
          <at 0.2937 1>
          <at 0.3503 0.4571>
         ]}
        RedCurve {keys [
          <at 0 0.95>
          <at 0.0435 0.95>
          {arriveTangent=-5.4821, interpMode=RCIM_Cubic, leaveTangent=-5.4821, time=0.1221, value=0.95}
          {arriveTangent=-11.7544, interpMode=RCIM_Cubic, leaveTangent=-11.7544, time=0.1438, value=0.4}
          <at 0.2029 0>
          <at 0.2454 0>
          <at 0.2937 0.4>
          <at 0.3503 0.5652>
         ]}
       }})
     Scale Factor=&MakeVector4FromVectorAndFloat("Vector (RGB)"=&VectorFromFloat(Value=0.8), "Float (Alpha)"=1)
    ))
  ]
  rendererClasses [&NiagaraSpriteRendererProperties]
  renderers [
   {rendererClass=&NiagaraSpriteRendererProperties}
   {rendererClass=&NiagaraSpriteRendererProperties, rendererIndex=1}
   {rendererClass=&NiagaraSpriteRendererProperties, rendererIndex=2}
  ]
 }
}
summary {
 systemName=NS_Stylizedpsplus
 emitters [
  {bEnabled=false, emitterName=NE_VC_Boost, rendererClasses [&NiagaraSpriteRendererProperties]}
  {bEnabled=false, emitterName=NE_VC_Boost_Sprite_Parent, rendererClasses [&NiagaraSpriteRendererProperties]}
  {emitterName=NE_VC_Boost_Sprite_Parent001, rendererClasses [&NiagaraSpriteRendererProperties]}
 ]
 userVariables [
  User.bIsFrontendPreview {type=NiagaraBool}
  User.bIsPainted {type=NiagaraBool}
  User.blsIsCoroColor {type=NiagaraBool}
  User.bPreviewVehicleSpeed {type=NiagaraBool}
  User.ColorVariantID {type=NiagaraInt32}
  User.DTCarOffset {type=Vector3f, description="", defaultValue {struct=Vector3f, value=<0 0 0>}}
  User.GameModeID {type=NiagaraInt32}
  User.PaintedColor {type=LinearColor, description="", defaultValue {struct=LinearColor, value=<rgba 1 1 1 1>}}
  User.PaintedScalar {type=NiagaraFloat, defaultValue=1, description=""}
  User.PortCount {type=NiagaraInt32, defaultValue=2}
 ]
}
